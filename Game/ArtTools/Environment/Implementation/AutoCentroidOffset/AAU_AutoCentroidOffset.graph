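# Editor utility Blueprint with an AutoCentroidOffset action that sets CentroidOffset on selected Blueprint assets from their static mesh.
# The function runs an embedded Python script that loads each selected Blueprint's class default object and reads its StaticMesh property.
# It writes the mesh's local bounding box center, cast to Vector3f, into CentroidOffset, then saves the asset.
# Non-Blueprint assets, unloadable defaults and empty StaticMesh values are skipped with a logged warning.
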EditorUtilityBlueprint /Game/ArtTools/Environment/Implementation/AutoCentroidOffset/AAU_AutoCentroidOffset
graphs [
 AutoCentroidOffset {nodes [
   AutoCentroidOffset() {is=title K2Node_FunctionEntry}
   > Execute Python Script(PythonScript="import unreal\r\n\r\ndef run():\r\n    selected_assets = unreal.EditorUtilityLibrary.get_selected_assets()\r\n\r\n    if not selected_assets:\r\n        unreal.log_warning(\"No assets selected. Please select Blueprint assets in the Content Browser.\")\r\n        return\r\n\r\n    for asset in selected_assets:\r\n        asset_name = asset.get_name()\r\n\r\n        if not isinstance(asset, unreal.Blueprint):\r\n            unreal.log_warning(f\"Skipping '{asset_name}' - not a Blueprint.\")\r\n            continue\r\n\r\n        ## Build the CDO object path: \"{AssetPath}.Default__{AssetName}_C\"\r\n        asset_path = asset.get_path_name().split(\".\")[0]\r\n        cdo_path = f\"{asset_path}.Default__{asset_name}_C\"\r\n\r\n        cdo = unreal.load_object(None, cdo_path)\r\n        if cdo is None:\r\n            unreal.log_warning(f\"Skipping '{asset_name}' - could not load CDO at: {cdo_path}\")\r\n            continue\r\n\r\n        static_mesh = cdo.get_editor_property(\"StaticMesh\")\r\n        if static_mesh is None:\r\n            unreal.log_warning(f\"Skipping '{asset_name}' - no StaticMesh property found.\")\r\n            continue\r\n\r\n        ## get_bounds().origin is the bounding box center in local space (double-precision Vector)\r\n        ## CentroidOffset is Vector3f (single-precision), so we must cast explicitly\r\n        origin = static_mesh.get_bounds().origin\r\n        center = unreal.Vector3f(origin.x, origin.y, origin.z)\r\n\r\n        cdo.set_editor_property(\"CentroidOffset\", center)\r\n\r\n        unreal.EditorAssetLibrary.save_loaded_asset(asset)\r\n\r\n        unreal.log(\r\n            f\"'{asset_name}': CentroidOffset set to \"\r\n            f\"({center.x:.2f}, {center.y:.2f}, {center.z:.2f})\"\r\n        )\r\n\r\nrun()\r\n")
  ]}
 EventGraph {links [], nodes []}
]
variables []
